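# Control Rig for the F_MED_HybridBoot female player body that procedurally corrects skirt, hip chain and other accessory bones.
# Spherical pose readers on both thighs store directional alphas, then a shared DeformRig function and weighted parent constraints for hip hoops, boot ties and jacket jewelry run.
# Those alphas, products of opposing-leg alphas, and remapped thigh Euler angles weight additive transform offsets on front skirt and hip chain bones.
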ControlRigBlueprint /BRCosmetics/Characters/Player/Female/Medium/Bodies/F_MED_HybridBoot/Meshes/F_MED_HybridBoot_CtrlRig
graphs [
 New Function_ContainedGraph {nodes [
   RigVMFunctionEntryNode()
   > &RigUnit_ParentConstraint(
    AdvancedSettings=(RotationOrderForFilter=XZY)
    Child=(Type=Bone,Name="base_hip_hoop_01_l")
    Parents=((Item=(Type=Bone,Name="spine_03"),Weight=0.0264),(Item=(Type=Bone,Name="spine_04"),Weight=0.1091),(Item=(Type=Bone,Name="spine_01"),Weight=0.2033),(Item=(Type=Bone,Name="pelvis"),Weight=0.2962),(Item=(Type=Bone,Name="spine_02"),Weight=0.365))
    Weight=1
   )
   > &RigUnit_ParentConstraint(
    AdvancedSettings=(RotationOrderForFilter=XZY)
    Child=(Type=Bone,Name="base_hip_hoop_01_r")
    Parents=((Item=(Type=Bone,Name="spine_03"),Weight=0.0264),(Item=(Type=Bone,Name="spine_04"),Weight=0.1091),(Item=(Type=Bone,Name="spine_01"),Weight=0.2033),(Item=(Type=Bone,Name="pelvis"),Weight=0.2962),(Item=(Type=Bone,Name="spine_02"),Weight=0.365))
    Weight=1
   )
   > &RigUnit_ParentConstraint(
    AdvancedSettings=(RotationOrderForFilter=XZY)
    Child=(Type=Bone,Name="base_tie_mstr_l")
    Parents=((Item=(Type=Bone,Name="calf_l"),Weight=0.99),(Item=(Type=Bone,Name="calf_twist_01_l"),Weight=0.144),(Item=(Type=Bone,Name="calf_twist_02_l"),Weight=0.756))
    Weight=1
   )
   > &RigUnit_ParentConstraint(
    AdvancedSettings=(RotationOrderForFilter=XZY)
    Child=(Type=Bone,Name="base_tie_mstr_r")
    Parents=((Item=(Type=Bone,Name="calf_r"),Weight=0.99),(Item=(Type=Bone,Name="calf_twist_01_r"),Weight=0.144),(Item=(Type=Bone,Name="calf_twist_02_r"),Weight=0.756))
    Weight=1
   )
   > &RigUnit_ParentConstraint(
    AdvancedSettings=(RotationOrderForFilter=XZY)
    Child=(Type=Bone,Name="dyn_jacketjewelry_front_l")
    Parents="((Item=(Type=Bone,Name=\"upperarm_l\"),Weight=0.0033),(Item=(Type=Bone,Name=\"deform_upperarm_up_l\"),Weight=0.0158),(Item=(Type=Bone,Name=\"deform_pec_l\"),Weight=0.0263),(Item=(Type=Bone,Name=\"spine_04\"),Weight=0.0748),(Item=(Type=Bone,Name=\"spine_05\"),Weight= ...(truncated)"
    Weight=1
   )
   > &RigUnit_ParentConstraint(
    AdvancedSettings=(RotationOrderForFilter=XZY)
    Child=(Type=Bone,Name="dyn_jacketjewelry_back_l")
    Parents="((Item=(Type=Bone,Name=\"clavicle_r\"),Weight=0.0004),(Item=(Type=Bone,Name=\"upperarm_l\"),Weight=0.0016),(Item=(Type=Bone,Name=\"deform_upperarm_up_l\"),Weight=0.0024),(Item=(Type=Bone,Name=\"spine_04\"),Weight=0.0114),(Item=(Type=Bone,Name=\"deform_pec_l\"),Weigh ...(truncated)"
    Weight=1
   )
   > &RigUnit_ParentConstraint(
    AdvancedSettings=(RotationOrderForFilter=XZY)
    Child=(Type=Bone,Name="dyn_jacketjewelry_front_r")
    Parents="((Item=(Type=Bone,Name=\"upperarm_r\"),Weight=0.0033),(Item=(Type=Bone,Name=\"deform_upperarm_up_r\"),Weight=0.0158),(Item=(Type=Bone,Name=\"deform_pec_r\"),Weight=0.0263),(Item=(Type=Bone,Name=\"spine_04\"),Weight=0.0748),(Item=(Type=Bone,Name=\"spine_05\"),Weight= ...(truncated)"
    Weight=1
   )
   > &RigUnit_ParentConstraint(
    AdvancedSettings=(RotationOrderForFilter=XZY)
    Child=(Type=Bone,Name="dyn_jacketjewelry_back_r")
    Parents="((Item=(Type=Bone,Name=\"clavicle_l\"),Weight=0.0004),(Item=(Type=Bone,Name=\"upperarm_r\"),Weight=0.0022),(Item=(Type=Bone,Name=\"deform_upperarm_up_r\"),Weight=0.0033),(Item=(Type=Bone,Name=\"spine_04\"),Weight=0.0122),(Item=(Type=Bone,Name=\"deform_pec_r\"),Weigh ...(truncated)"
    Weight=1
   )
   > RigVMFunctionReturnNode()
   &RigUnit_ParentConstraint(
    AdvancedSettings=(RotationOrderForFilter=XZY)
    Child=(Type=Bone,Name="base_hip_hoop_01_l")
    ExecutePin=()
    Parents=((Item=(Type=Bone,Name="spine_01"),Weight=0.19),(Item=(Type=Bone,Name="spine_03"),Weight=0.011),(Item=(Type=Bone,Name="spine_04"),Weight=0.86),(Item=(Type=Bone,Name="spine_02"),Weight=0.33),(Item=(Type=Bone,Name="pelvis"),Weight=0.37))
    Weight=1
   )
   > &RigUnit_ParentConstraint(
    AdvancedSettings=(RotationOrderForFilter=XZY)
    Child=(Type=Bone,Name="base_hip_hoop_01_r")
    Parents=((Item=(Type=Bone,Name="spine_01"),Weight=0.19),(Item=(Type=Bone,Name="spine_03"),Weight=0.011),(Item=(Type=Bone,Name="spine_04"),Weight=0.86),(Item=(Type=Bone,Name="spine_02"),Weight=0.33),(Item=(Type=Bone,Name="pelvis"),Weight=0.37))
    Weight=1
   )
  ]}
 New Function_ContainedGraph {nodes [
   n1 = RigVMFunctionEntryNode()
   RigVMFunction_MathQuaternionToEuler = &RigVMFunction_MathQuaternionToEuler(
    RotationOrder=XYZ
    Value=&RigUnit_GetRelativeTransformForItem(
     Child=(Type=Bone,Name="hand_l")
     Parent=(Type=Null,Name="lowerarm_l")
     bChildInitial=False
     bParentInitial=False
     Child.Name=@n1.Child
     Parent.Name=@n1.Parent
    ).RelativeTransform.Rotation
   )
   Remap = &RigVMFunction_MathDoubleRemap(
    TargetMaximum=1
    TargetMinimum=0
    Value=@RigVMFunction_MathQuaternionToEuler.Result.Y
    SourceMaximum=@n1.Source Maximum Y
    SourceMinimum=@n1.Source Minimum Y
   )
   > RigVMFunctionReturnNode(Delta Rotation Y=@Remap.Result)
  ]}
 New Function_ContainedGraph {nodes [
   n1 = RigVMFunctionEntryNode()
   RigVMFunction_MathQuaternionToEuler = &RigVMFunction_MathQuaternionToEuler(
    RotationOrder=XYZ
    Value=&RigUnit_GetRelativeTransformForItem(
     Child=(Type=Bone,Name="hand_l")
     Parent=(Type=Null,Name="lowerarm_l")
     bChildInitial=False
     bParentInitial=False
     Child.Name=@n1.Child
     Parent.Name=@n1.Parent
    ).RelativeTransform.Rotation
   )
   > RigVMFunctionReturnNode(Delta Rotation Z=&RigVMFunction_MathDoubleRemap(
     TargetMaximum=1
     TargetMinimum=0
     Value=@RigVMFunction_MathQuaternionToEuler.Result.Z
     SourceMaximum=@n1.Source Maximum Z
     SourceMinimum=@n1.Source Minimum Z
    ).Result)
  ]}
 RigVMFunctionLibrary {nodes [
   Constraints = RigVMCollapseNode(ExecuteContext=())
   Delta Rotation Y = RigVMCollapseNode(ExecuteContext=())
   Delta Rotation Z = RigVMCollapseNode(ExecuteContext=())
  ]}
 RigVMFunction_Sequence_1_ContainedGraph {nodes [
   n1 = RigVMFunctionEntryNode()
   RigVMFunction_Sequence = &RigVMFunction_Sequence(ExecuteContext=@n1.ExecuteContext)
   n2 = &RigVMFunction_Sequence(ExecuteContext=@RigVMFunction_Sequence.B)
   n3 = &RigVMFunction_Sequence(ExecuteContext=@n2.B)
   n4 = &RigVMFunction_Sequence(ExecuteContext=@n3.B)
   RigVMFunctionReturnNode(A=@RigVMFunction_Sequence.A, B=@n2.A, C=@n3.A, D=@n4.A, E=@n4.B)
  ]}
 RigVMFunction_Sequence_2_ContainedGraph {nodes [
   n1 = RigVMFunctionEntryNode()
   n2 = &RigVMFunction_Sequence(ExecuteContext=@n1.ExecuteContext)
   n3 = &RigVMFunction_Sequence(ExecuteContext=@n2.B)
   RigVMFunctionReturnNode(A=@n2.A, B=@n3.A, C=@n3.B)
  ]}
 RigVMModel {nodes [
   RigUnit_BeginExecution = &RigUnit_BeginExecution()
   RigVMFunction_Sequence_1 = RigVMAggregateNode(ExecuteContext=@RigUnit_BeginExecution.ExecutePin)
   n1 = &RigUnit_SphericalPoseReader(
    ActiveRegionScaleFactors=(PositiveWidth=4,NegativeWidth=3.5,PositiveHeight=2.5,NegativeHeight=5)
    ActiveRegionSize=0.1
    Debug=(bDrawDebug=true,bDraw2D=False,bDrawLocalAxes=False,DebugScale=25,DebugSegments=20,DebugThickness=0.25)
    DriverAxis=(X=-1,Y=0,Z=0)
    DriverItem=(Type=Bone,Name="thigh_l")
    FalloffRegionScaleFactors=(PositiveWidth=1,NegativeWidth=3,PositiveHeight=1,NegativeHeight=0.5)
    FalloffSize=0.4
    FlipHeightScaling=False
    FlipWidthScaling=False
    OptionalParentItem=(Type=Bone,Name="None")
    RotationOffset=(X=100,Y=0,Z=-20)
    ExecutePin=@RigVMFunction_Sequence_1.A
   )
   > RigVMVariableNode(Value=@n1.OutputParam) {variable=L thighForward Alpha}
   > n2 = &RigUnit_SphericalPoseReader(
    ActiveRegionScaleFactors=(PositiveWidth=4,NegativeWidth=3.5,PositiveHeight=2.5,NegativeHeight=5)
    ActiveRegionSize=0.1
    Debug=(bDrawDebug=true,bDraw2D=False,bDrawLocalAxes=False,DebugScale=25,DebugSegments=20,DebugThickness=0.25)
    DriverAxis=(X=1,Y=0,Z=0)
    DriverItem=(Type=Bone,Name="thigh_r")
    FalloffRegionScaleFactors=(PositiveWidth=1,NegativeWidth=3,PositiveHeight=1,NegativeHeight=0.5)
    FalloffSize=0.4
    FlipHeightScaling=False
    FlipWidthScaling=true
    OptionalParentItem=(Type=Bone,Name="None")
    RotationOffset=(X=-80,Y=0,Z=-20)
   )
   > RigVMVariableNode(Value=@n2.OutputParam) {variable=R thighForward Alpha}
   > n3 = &RigUnit_SphericalPoseReader(
    ActiveRegionScaleFactors=(PositiveWidth=3,NegativeWidth=2,PositiveHeight=2,NegativeHeight=2)
    ActiveRegionSize=0.1
    Debug=(bDrawDebug=false,bDraw2D=False,bDrawLocalAxes=False,DebugScale=25,DebugSegments=20,DebugThickness=0.25)
    DriverAxis=(X=-1,Y=0,Z=0)
    DriverItem=(Type=Bone,Name="thigh_l")
    FalloffRegionScaleFactors=(PositiveWidth=2,NegativeWidth=1,PositiveHeight=2,NegativeHeight=2)
    FalloffSize=0.25
    FlipHeightScaling=false
    FlipWidthScaling=false
    OptionalParentItem=(Type=Bone,Name="None")
    RotationOffset=(X=180,Y=-40,Z=0)
   )
   > RigVMVariableNode(Value=@n3.OutputParam) {variable=L thighSide Alpha}
   > n4 = &RigUnit_SphericalPoseReader(
    ActiveRegionScaleFactors=(PositiveWidth=3,NegativeWidth=2,PositiveHeight=2,NegativeHeight=2)
    ActiveRegionSize=0.1
    Debug=(bDrawDebug=false,bDraw2D=False,bDrawLocalAxes=False,DebugScale=25,DebugSegments=20,DebugThickness=0.25)
    DriverAxis=(X=1,Y=0,Z=0)
    DriverItem=(Type=Bone,Name="thigh_r")
    FalloffRegionScaleFactors=(PositiveWidth=2,NegativeWidth=1,PositiveHeight=2,NegativeHeight=2)
    FalloffSize=0.25
    FlipHeightScaling=False
    FlipWidthScaling=true
    OptionalParentItem=(Type=Bone,Name="None")
    RotationOffset=(X=0,Y=-40,Z=0)
   )
   > RigVMVariableNode(Value=@n4.OutputParam) {variable=R thighSide Alpha}
   > n5 = &RigUnit_SphericalPoseReader(
    ActiveRegionScaleFactors=(PositiveWidth=2,NegativeWidth=2,PositiveHeight=2,NegativeHeight=1)
    ActiveRegionSize=0.1
    Debug=(bDrawDebug=false,bDraw2D=False,bDrawLocalAxes=False,DebugScale=25,DebugSegments=20,DebugThickness=0.25)
    DriverAxis=(X=-1,Y=0,Z=0)
    DriverItem=(Type=Bone,Name="thigh_l")
    FalloffRegionScaleFactors=(PositiveWidth=2,NegativeWidth=2,PositiveHeight=4,NegativeHeight=1)
    FalloffSize=0.3
    FlipHeightScaling=false
    FlipWidthScaling=false
    OptionalParentItem=(Type=Bone,Name="None")
    RotationOffset=(X=0,Y=90,Z=80)
   )
   > RigVMVariableNode(Value=@n5.OutputParam) {variable=L thighBack Alpha}
   > n6 = &RigUnit_SphericalPoseReader(
    ActiveRegionScaleFactors=(PositiveWidth=2,NegativeWidth=2,PositiveHeight=2,NegativeHeight=1)
    ActiveRegionSize=0.1
    Debug=(bDrawDebug=false,bDraw2D=False,bDrawLocalAxes=False,DebugScale=25,DebugSegments=20,DebugThickness=0.25)
    DriverAxis=(X=1,Y=0,Z=0)
    DriverItem=(Type=Bone,Name="thigh_r")
    FalloffRegionScaleFactors=(PositiveWidth=2,NegativeWidth=2,PositiveHeight=3,NegativeHeight=1)
    FalloffSize=0.3
    FlipHeightScaling=false
    FlipWidthScaling=false
    OptionalParentItem=(Type=Bone,Name="None")
    RotationOffset=(X=180,Y=93.7803,Z=60.6503)
   )
   > RigVMVariableNode(Value=@n6.OutputParam) {variable=R thighBack Alpha}
   n7 = &RigUnit_SphericalPoseReader(
    ActiveRegionScaleFactors=(PositiveWidth=1.5,NegativeWidth=1.5,PositiveHeight=1.5,NegativeHeight=1.5)
    ActiveRegionSize=0.1
    Debug=(bDrawDebug=true,bDraw2D=False,bDrawLocalAxes=False,DebugScale=25,DebugSegments=20,DebugThickness=0.25)
    DriverAxis=(X=-1,Y=0,Z=0)
    DriverItem=(Type=Bone,Name="thigh_l")
    FalloffRegionScaleFactors=(PositiveWidth=2,NegativeWidth=2,PositiveHeight=2,NegativeHeight=2)
    FalloffSize=0.23
    FlipHeightScaling=false
    FlipWidthScaling=false
    OptionalParentItem=(Type=Bone,Name="None")
    RotationOffset=(X=33.1118,Y=48.7758,Z=-2.2671)
    ExecutePin=@RigVMFunction_Sequence_1.B
   )
   > n8 = &RigUnit_SphericalPoseReader(
    ActiveRegionScaleFactors=(PositiveWidth=1.5,NegativeWidth=1.5,PositiveHeight=1.5,NegativeHeight=1.5)
    ActiveRegionSize=0.1
    Debug=(bDrawDebug=false,bDraw2D=False,bDrawLocalAxes=False,DebugScale=25,DebugSegments=20,DebugThickness=0.25)
    DriverAxis=(X=1,Y=0,Z=0)
    DriverItem=(Type=Bone,Name="thigh_r")
    FalloffRegionScaleFactors=(PositiveWidth=2,NegativeWidth=2,PositiveHeight=2,NegativeHeight=2)
    FalloffSize=0.23
    FlipHeightScaling=False
    FlipWidthScaling=true
    OptionalParentItem=(Type=Bone,Name="None")
    RotationOffset=(X=-166.5537,Y=44.1032,Z=13.3453)
   )
   > n9 = &RigUnit_SphericalPoseReader(
    ActiveRegionScaleFactors=(PositiveWidth=1,NegativeWidth=1.2,PositiveHeight=1,NegativeHeight=1)
    ActiveRegionSize=0.13
    Debug=(bDrawDebug=false,bDraw2D=False,bDrawLocalAxes=False,DebugScale=25,DebugSegments=20,DebugThickness=0.25)
    DriverAxis=(X=-1,Y=0,Z=0)
    DriverItem=(Type=Bone,Name="thigh_l")
    FalloffRegionScaleFactors=(PositiveWidth=1,NegativeWidth=1.2,PositiveHeight=1,NegativeHeight=1)
    FalloffSize=0.25
    FlipHeightScaling=false
    FlipWidthScaling=false
    OptionalParentItem=(Type=Bone,Name="None")
    RotationOffset=(X=99.0839,Y=-161.4364,Z=74.6992)
   )
   > n10 = &RigUnit_SphericalPoseReader(
    ActiveRegionScaleFactors=(PositiveWidth=1,NegativeWidth=1,PositiveHeight=1,NegativeHeight=1)
    ActiveRegionSize=0.12
    Debug=(bDrawDebug=false,bDraw2D=False,bDrawLocalAxes=False,DebugScale=25,DebugSegments=20,DebugThickness=0.25)
    DriverAxis=(X=1,Y=0,Z=0)
    DriverItem=(Type=Bone,Name="thigh_r")
    FalloffRegionScaleFactors=(PositiveWidth=1,NegativeWidth=1,PositiveHeight=1,NegativeHeight=1)
    FalloffSize=0.25
    FlipHeightScaling=False
    FlipWidthScaling=true
    OptionalParentItem=(Type=Bone,Name="None")
    RotationOffset=(X=-108.1179,Y=53.0444,Z=19.1451)
   )
   > RigVMVariableNode(Value=@n7.OutputParam) {variable=L thighSideIn_Fwd Alpha}
   > RigVMVariableNode(Value=@n8.OutputParam) {variable=R thighSideIn_Bwd Alpha}
   > RigVMVariableNode(Value=@n9.OutputParam) {variable=L thighSideIn_Bwd Alpha}
   > RigVMVariableNode(Value=@n10.OutputParam) {variable=R thighSideIn_Fwd Alpha}
   FN_DeformRig_FNC = RigVMFunctionReferenceNode(
    DeformRig_ Curve="(deformRig_l_upperarm_up_alpha_72_#1=true,deformRig_r_upperarm_up_alpha_73_#2=true,deformRig_l_upperarm_fr_alpha_74_#3=true,deformRig_r_upperarm_fr_alpha_75_CB51B80C4 ...(truncated)"
    DeformRig_Bool="(upperarm_1_#4=True,elbow_3_#5=True,wrist_7_#6=True,groin_8_#7=True,glute_17_#8=True,knee_9_40C408324CA261 ...(truncated)"
    DeformRig_Debug=(upperarm_debug_18_#4=False,wrist_debug_20_#6=False,lat_debug_26_#9=False,groinGlute_debug_27_#7=False)
    DeformRig_Value="(upperarm_up_18_#10=(Rotation=(X=0,Y=0,Z=0,W=1),Translation=(X=0,Y=0,Z=8),Scale3D=(X=0.7,Y=1,Z=1)),upperarm_fwd_fr_19_#11=(Rotation=(X=0,Y=0,Z=0,W=1),Translation=(X=0,Y=10,Z=0),Scale3D=(X=0.5,Y=1,Z= ...(truncated)"
    ExecuteContext=@RigVMFunction_Sequence_1.C
   )
   > Constraints = RigVMFunctionReferenceNode()
   > RigVMFunction_Sequence = RigVMAggregateNode()
   Delta Rotation Y_1 = RigVMFunctionReferenceNode(Child=thigh_r, Parent=thigh_r_null, Source Maximum Y=-90, Source Minimum Y=-30, ExecuteContext=@RigVMFunction_Sequence.A)
   > &RigUnit_ModifyTransforms(
    ItemToModify=((Item=(Type=Bone,Name="dyn_hip_side_chain_02_r"),Transform=(Rotation=(X=0,Y=0,Z=-0.5,W=0.866),Translation=(X=0,Y=0,Z=0),Scale3D=(X=1,Y=1,Z=1))))
    WeightMinimum=0
    Weight=@"Delta Rotation Y_1".Delta Rotation Y
   )
   > &RigUnit_ModifyTransforms(
    ItemToModify=((Item=(Type=Bone,Name="dyn_hip_side_chain_01_r"),Transform=(Rotation=(X=0,Y=0,Z=-0.0872,W=0.9962),Translation=(X=0,Y=0,Z=0),Scale3D=(X=1,Y=1,Z=1))))
    WeightMinimum=0
    Weight=@"Delta Rotation Y_1".Delta Rotation Y
   )
   > Delta Rotation Z_1 = RigVMFunctionReferenceNode(Child=thigh_r, Parent=thigh_r_null, Source Maximum Z=-90, Source Minimum Z=-50)
   > &RigUnit_ModifyTransforms(
    ItemToModify=((Item=(Type=Bone,Name="dyn_hip_side_chain_02_r"),Transform=(Rotation=(X=0.13,Y=-0.0114,Z=-0.0864,W=0.9877),Translation=(X=0,Y=0,Z=0),Scale3D=(X=1,Y=1,Z=1))))
    WeightMinimum=0
    Weight=@"Delta Rotation Z_1".Delta Rotation Z
   )
   > &RigUnit_ModifyTransforms(
    ItemToModify=((Item=(Type=Bone,Name="dyn_hip_side_chain_01_r"),Transform=(Rotation=(X=0.0704,Y=-0.1006,Z=-0.1648,W=0.9786),Translation=(X=0,Y=0,Z=0),Scale3D=(X=1,Y=1,Z=1))))
    WeightMinimum=0
    Weight=@"Delta Rotation Z_1".Delta Rotation Z
   )
   Delta Rotation Y_3 = RigVMFunctionReferenceNode(Child=thigh_l, Parent=thigh_l_null, Source Maximum Y=90, Source Minimum Y=-30, ExecuteContext=@RigVMFunction_Sequence.B)
   > &RigUnit_ModifyTransforms(
    ItemToModify=((Item=(Type=Bone,Name="dyn_hip_side_chain_02_l"),Transform=(Rotation=(X=0,Y=0,Z=-0.5,W=0.866),Translation=(X=0,Y=0,Z=0),Scale3D=(X=1,Y=1,Z=1))))
    WeightMinimum=0
    Weight=@"Delta Rotation Y_3".Delta Rotation Y
   )
   > &RigUnit_ModifyTransforms(
    ItemToModify=((Item=(Type=Bone,Name="dyn_hip_side_chain_01_l"),Transform=(Rotation=(X=0,Y=0,Z=-0.0872,W=0.9962),Translation=(X=0,Y=0,Z=0),Scale3D=(X=1,Y=1,Z=1))))
    WeightMinimum=0
    Weight=@"Delta Rotation Y_3".Delta Rotation Y
   )
   > Delta Rotation Z = RigVMFunctionReferenceNode(Child=thigh_l, Parent=thigh_l_null, Source Maximum Z=-90, Source Minimum Z=-60)
   > &RigUnit_ModifyTransforms(
    ItemToModify=((Item=(Type=Bone,Name="dyn_hip_side_chain_02_l"),Transform=(Rotation=(X=0.13,Y=-0.0114,Z=-0.0864,W=0.9877),Translation=(X=0,Y=0,Z=0),Scale3D=(X=1,Y=1,Z=1))))
    WeightMinimum=0
    Weight=@"Delta Rotation Z".Delta Rotation Z
   )
   > &RigUnit_ModifyTransforms(
    ItemToModify=((Item=(Type=Bone,Name="dyn_hip_side_chain_01_l"),Transform=(Rotation=(X=0.0704,Y=-0.1006,Z=-0.1648,W=0.9786),Translation=(X=-1,Y=0,Z=0),Scale3D=(X=1,Y=1,Z=1))))
    WeightMinimum=0
    Weight=@"Delta Rotation Z".Delta Rotation Z
   )
   &RigUnit_RotationConstraintLocalSpaceOffset(
    AdvancedSettings=(InterpolationType=Average,RotationOrderForFilter=XZY)
    Child=(Type=Bone,Name="base_front_skirt_Inner_r")
    Filter=(bX=true,bY=True,bZ=True)
    Parents=((Item=(Type=Bone,Name="thigh_r"),Weight=1))
    Weight=1
    ExecutePin=@RigVMFunction_Sequence.C
   )
   > &RigUnit_RotationConstraintLocalSpaceOffset(
    AdvancedSettings=(InterpolationType=Average,RotationOrderForFilter=XZY)
    Child=(Type=Bone,Name="base_front_skirt_inner_l")
    Filter=(bX=true,bY=True,bZ=True)
    Parents=((Item=(Type=Bone,Name="thigh_l"),Weight=1))
    Weight=1
   )
   n11 = RigVMVariableNode() {variable=L thighForward Alpha}
   > &RigUnit_ModifyTransforms(
    ItemToModify=((Item=(Type=Bone,Name="base_front_skirt_inner_l"),Transform=(Rotation=(X=-0.0042,Y=-0.0958,Z=-0.0434,W=0.9944),Translation=(X=0,Y=0,Z=1),Scale3D=(X=1,Y=1,Z=1))))
    WeightMinimum=0
    Weight=@n11.Value
   )
   > &RigUnit_ModifyTransforms(
    ItemToModify=((Item=(Type=Bone,Name="base_front_skirt_splitter_l"),Transform=(Rotation=(X=0.0003,Y=-0.0349,Z=0.0087,W=0.9994),Translation=(X=0,Y=0,Z=0.5),Scale3D=(X=1,Y=1,Z=1))))
    WeightMinimum=0
    Weight=@n11.Value
   )
   > &RigUnit_ModifyTransforms(
    ItemToModify=((Item=(Type=Bone,Name="dyn_front_skirt_splitter_length_l"),Transform=(Rotation=(X=0,Y=0,Z=0,W=1),Translation=(X=1,Y=0,Z=0),Scale3D=(X=1,Y=1,Z=1))))
    WeightMinimum=0
    Weight=@n11.Value
   )
   > &RigUnit_ModifyTransforms(
    ItemToModify=((Item=(Type=Bone,Name="dyn_front_skirt_Inner_length_l"),Transform=(Rotation=(X=0,Y=0,Z=0,W=1),Translation=(X=-1,Y=0,Z=0),Scale3D=(X=1,Y=1,Z=1))))
    WeightMinimum=0
    Weight=@n11.Value
   )
   n12 = RigVMVariableNode() {variable=R thighForward Alpha}
   > &RigUnit_ModifyTransforms(
    ItemToModify=((Item=(Type=Bone,Name="base_front_skirt_Inner_r"),Transform=(Rotation=(X=0.0084,Y=-0.0955,Z=0.0868,W=0.9916),Translation=(X=0,Y=0,Z=1),Scale3D=(X=1,Y=1,Z=1))))
    WeightMinimum=0
    Weight=@n12.Value
   )
   > &RigUnit_ModifyTransforms(
    ItemToModify=((Item=(Type=Bone,Name="base_front_skirt_splitter_r"),Transform=(Rotation=(X=0.0003,Y=0.0349,Z=-0.0087,W=0.9994),Translation=(X=0,Y=0,Z=0.5),Scale3D=(X=1,Y=1,Z=1))))
    WeightMinimum=0
    Weight=@n12.Value
   )
   > &RigUnit_ModifyTransforms(
    ItemToModify=((Item=(Type=Bone,Name="dyn_front_skirt_splitter_length_r"),Transform=(Rotation=(X=0,Y=0,Z=0,W=1),Translation=(X=-1,Y=0,Z=0),Scale3D=(X=1,Y=1,Z=1))))
    WeightMinimum=0
    Weight=@n12.Value
   )
   > &RigUnit_ModifyTransforms(
    ItemToModify=((Item=(Type=Bone,Name="dyn_front_skirt_Inner_length_r"),Transform=(Rotation=(X=0,Y=0,Z=0,W=1),Translation=(X=-1,Y=0,Z=0),Scale3D=(X=1,Y=1,Z=1))))
    WeightMinimum=0
    Weight=@n12.Value
   )
   > &RigUnit_ModifyTransforms(
    ItemToModify=((Item=(Type=Bone,Name="base_front_skirt_inner_l"),Transform=(Rotation=(X=0,Y=0,Z=-0.0872,W=0.9962),Translation=(X=0,Y=0,Z=0),Scale3D=(X=1,Y=1,Z=1))))
    WeightMinimum=0
    Weight=RigVMVariableNode() {variable=L thighSide Alpha}.Value
   )
   > &RigUnit_ModifyTransforms(
    ItemToModify=((Item=(Type=Bone,Name="base_front_skirt_Inner_r"),Transform=(Rotation=(X=0,Y=0,Z=0.0872,W=0.9962),Translation=(X=0,Y=0,Z=0),Scale3D=(X=1,Y=1,Z=1))))
    WeightMinimum=0
    Weight=RigVMVariableNode() {variable=R thighSide Alpha}.Value
   )
   > &RigUnit_ModifyTransforms(
    ItemToModify=((Item=(Type=Bone,Name="base_front_skirt_inner_l"),Transform=(Rotation=(X=0.0076,Y=-0.1735,Z=0.043,W=0.9839),Translation=(X=0,Y=0,Z=1),Scale3D=(X=1,Y=1,Z=1))))
    WeightMinimum=0
    Weight=RigVMVariableNode() {variable=L thighBack Alpha}.Value
   )
   > &RigUnit_ModifyTransforms(
    ItemToModify=((Item=(Type=Bone,Name="base_front_skirt_Inner_r"),Transform=(Rotation=(X=0.0151,Y=-0.173,Z=0.0858,W=0.9811),Translation=(X=0,Y=0,Z=1),Scale3D=(X=1,Y=1,Z=1))))
    WeightMinimum=0
    Weight=RigVMVariableNode() {variable=R thighBack Alpha}.Value
   )
   n13 = &RigVMFunction_MathDoubleMul(A=RigVMVariableNode() {variable=L thighSideIn_Fwd Alpha}.Value, B=RigVMVariableNode() {variable=R thighSideIn_Bwd Alpha}.Value)
   > n14 = &RigUnit_ModifyTransforms(
    ItemToModify=((Item=(Type=Bone,Name="base_front_skirt_Inner_r"),Transform=(Rotation=(X=0.0151,Y=-0.173,Z=0.0858,W=0.9811),Translation=(X=0,Y=0,Z=1),Scale3D=(X=1,Y=1,Z=1))))
    WeightMinimum=0
    Weight=@n13.Result
   )
   Less = /Script/RigVM.RigVMFunction_MathDoubleLess(B=1, A=@n13.Result) {is=struct RigVMUnitNode, template="Less::Execute(in A,in B,out Result)"}
   RigVMFunction_ControlFlowBranch = &RigVMFunction_ControlFlowBranch(Condition=@Less.Result, ExecuteContext=@n14.ExecutePin)
   &RigUnit_ModifyTransforms(
    ItemToModify=((Item=(Type=Bone,Name="base_front_skirt_inner_l"),Transform=(Rotation=(X=0.0076,Y=-0.1735,Z=0.043,W=0.9839),Translation=(X=0,Y=0,Z=1),Scale3D=(X=1,Y=1,Z=1))))
    WeightMinimum=0
    Weight=&RigVMFunction_MathDoubleMul(A=RigVMVariableNode() {variable=L thighSideIn_Bwd Alpha}.Value, B=RigVMVariableNode() {variable=R thighSideIn_Fwd Alpha}.Value).Result
    ExecutePin=@RigVMFunction_ControlFlowBranch.True
   )
   Delta Rotation Y = RigVMFunctionReferenceNode(Child=thigh_r, Parent=thigh_r_null, Source Maximum Y=-100, Source Minimum Y=-30, ExecuteContext=@RigVMFunction_Sequence.E)
   > &RigUnit_ModifyTransforms(
    ItemToModify=((Item=(Type=Bone,Name="base_front_skirt_splitter_r"),Transform=(Rotation=(X=0,Y=0,Z=-0.2588,W=0.9659),Translation=(X=0,Y=0,Z=0),Scale3D=(X=1,Y=1,Z=1))))
    WeightMinimum=0
    Weight=@"Delta Rotation Y".Delta Rotation Y
   )
   > Delta Rotation Y_2 = RigVMFunctionReferenceNode(Child=thigh_l, Parent=thigh_l_null, Source Maximum Y=100, Source Minimum Y=30)
   > &RigUnit_ModifyTransforms(
    ItemToModify=((Item=(Type=Bone,Name="base_front_skirt_splitter_l"),Transform=(Rotation=(X=0,Y=0,Z=0.2588,W=0.9659),Translation=(X=0,Y=0,Z=0),Scale3D=(X=1,Y=1,Z=1))))
    WeightMinimum=0
    Weight=@"Delta Rotation Y_2".Delta Rotation Y
   )
   &RigUnit_ModifyTransforms(
    ItemToModify=((Item=(Type=Bone,Name="base_front_skirt_inner_l"),Transform=(Rotation=(X=-0.0034,Y=-0.0784,Z=-0.0435,W=0.996),Translation=(X=0,Y=0,Z=1),Scale3D=(X=1,Y=1,Z=1))))
    WeightMinimum=0
    Weight=RigVMVariableNode() {variable=L thighSideIn Alpha}.Value
    ExecutePin=@RigVMFunction_Sequence.D
   )
   > &RigUnit_ModifyTransforms(
    ItemToModify=((Item=(Type=Bone,Name="base_front_skirt_Inner_r"),Transform=(Rotation=(X=0,Y=0,Z=0.1736,W=0.9848),Translation=(X=0,Y=0,Z=1),Scale3D=(X=1,Y=1,Z=1))))
    WeightMinimum=0
    Weight=RigVMVariableNode() {variable=R thighSideIn Alpha}.Value
   )
   RigVMFunction_MathBoolNot = &RigVMFunction_MathBoolNot(Value=RigVMVariableNode() {variable=isCrouching}.Value)
   RigVMFunction_MathBoolAnd = &RigVMFunction_MathBoolAnd(A=@Less.Result, B=@RigVMFunction_MathBoolNot.Result)
  ]}
]
hierarchy [
 root {loc=<0 0 0>, rot=<0 0 0 1>}
 pelvis {parent=root, loc=<0 -0.5301 98.0571>, rot=<0.0127 0.707 -0.0127 -0.707>}
 spine_01 {parent=pelvis, loc=<6.1208 -0.9328 0>, rot=<0 0 0.034 -0.9994>}
 spine_02 {parent=spine_01, loc=<6.5997 -0.2095 0>, rot=<0 0 0.0238 -0.9997>}
 spine_03 {parent=spine_02, loc=<6.4914 -0.1237 0>, rot=<0 0 0.0149 -0.9999>}
 spine_04 {parent=spine_03, loc=<6.9598 0.0794 0>, rot=<0 0 -0.0132 -0.9999>}
 spine_05 {parent=spine_04, loc=<6.9987 0 0>, rot=<0 0 -0.0015 -1>}
 neck_01 {parent=spine_05, loc=<11.3138 0.1529 0>, rot=<0 0 -0.2003 -0.9797>}
 neck_02 {parent=neck_01, loc=<4.5627 0 0>, rot=<0 0 0.0504 -0.9987>}
 head {parent=neck_02, loc=<5.3914 0 0>, rot=<0 0 0.1011 -0.9949>}
 clavicle_l {parent=spine_05, loc=<5.4831 -0.1053 -3.2946>, rot=<0.0465 -0.7289 -0.009 -0.6829>}
 upperarm_l {parent=clavicle_l, loc=<9.8658 -1.2136 -0.4105>, rot=<0.0164 -0.3245 -0.0269 -0.9454>}
 lowerarm_l {parent=upperarm_l, loc=<26.465 -0.8345 0.259>, rot=<0.0036 0.0161 -0.2298 -0.9731>}
 hand_l {parent=lowerarm_l, loc=<22.7355 -0.3203 -0.6105>, rot=<-0.6784 -0.0223 0.0735 -0.7307>}
 thumb_01_l {parent=hand_l, loc=<1.3118 -0.3191 -2.459>, rot=<0.6546 -0.4042 0.0664 -0.6354>}
 thumb_02_l {parent=thumb_01_l, loc=<5.9336 -0.0448 0.0401>, rot=<-0.0037 0.0213 0.298 -0.9543>}
 thumb_03_l {parent=thumb_02_l, loc=<2.2223 0.3682 -0.0907>, rot=<0.0118 -0.013 0.0066 -0.9998>}
 index_metacarpal_l {parent=hand_l, loc=<1.6535 0.0183 -2.96>, rot=<0.0158 -0.0124 -0.0173 -0.9996>}
 index_01_l {parent=index_metacarpal_l, loc=<8.3008 0.0357 -0.0524>, rot=<-0.0044 0.0485 0.0924 -0.9945>}
 index_02_l {parent=index_01_l, loc=<3.3587 -0.3004 0.0452>, rot=<-0.002 0.0027 0.3225 -0.9466>}
 index_03_l {parent=index_02_l, loc=<2.5578 -0.123 -0.0252>, rot=<-0.002 0.0067 0.2558 -0.9667>}
 middle_metacarpal_l {parent=hand_l, loc=<2.1856 -0.1725 -0.5984>, rot=<0.006 0.0168 -0.0329 -0.9993>}
 middle_01_l {parent=middle_metacarpal_l, loc=<7.3275 -0.2447 0.0129>, rot=<-0.0208 0.0042 0.0858 -0.9961>}
 middle_02_l {parent=middle_01_l, loc=<4.0123 0.1401 0.1175>, rot=<-0.0218 0.0258 0.3512 -0.9357>}
 middle_03_l {parent=middle_02_l, loc=<2.6077 0.0534 0.1373>, rot=<0.0082 0.0188 0.1708 -0.9851>}
 ring_metacarpal_l {parent=hand_l, loc=<2.2074 -0.3625 1.3638>, rot=<-0.0284 0.0722 -0.0458 -0.9959>}
 ring_01_l {parent=ring_metacarpal_l, loc=<7.0295 -0.6197 -0.1446>, rot=<-0.0489 -0.0252 0.1892 -0.9804>}
 ring_02_l {parent=ring_01_l, loc=<3.6174 0.4647 -0.1027>, rot=<0 0 0.2806 -0.9598>}
 ring_03_l {parent=ring_02_l, loc=<2.4152 0.114 0.0199>, rot=<0 0 0.1189 -0.9929>}
 pinky_metacarpal_l {parent=hand_l, loc=<1.7616 -0.9753 3.3814>, rot=<-0.0397 0.0814 -0.069 -0.9935>}
 pinky_01_l {parent=pinky_metacarpal_l, loc=<7.1709 -0.4831 -0.1094>, rot=<-0.0654 -0.0035 0.2242 -0.9723>}
 pinky_02_l {parent=pinky_01_l, loc=<2.8274 0.0893 -0.1043>, rot=<0.0037 0.0011 0.2854 -0.9584>}
 pinky_03_l {parent=pinky_02_l, loc=<2.2046 0.0307 -0.0078>, rot=<-0.0002 0.0098 0.1445 -0.9895>}
 hand_attach_l {parent=hand_l, loc=<8.5 -3 0>, rot=<0.7071 0.7071 0 0>}
 weapon_l {parent=hand_l, loc=<-0.1974 -3.4111 0.3811>, rot=<0 0 0 1>}
 deform_wrist_dn_base_l {parent=hand_l, loc=<0 -1.5326 0>, rot=<0 0 0 -1>}
 deform_wrist_dn_l {parent=deform_wrist_dn_base_l, loc=<0 0 0>, rot=<0 0 0 1>}
 deform_wrist_up_base_l {parent=hand_l, loc=<0 2.3009 0>, rot=<0 0 0 -1>}
 deform_wrist_up_l {parent=deform_wrist_up_base_l, loc=<0 0 0>, rot=<0 0 0 1>}
 lowerarm_twist_02_l {parent=lowerarm_l, loc=<10.6707 0 0>, rot=<0 0 0 1>}
 lowerarm_twist_01_l {parent=lowerarm_l, loc=<18.316 0 0>, rot=<0 0 0 1>}
 upperarm_twist_01_l {parent=upperarm_l, loc=<7.2135 -0.1134 0>, rot=<0 0 0 1>}
 deform_upperarm_bk_base_l {parent=upperarm_twist_01_l, loc=<-6.4053 -4.3746 0.314>, rot=<0 0 0 -1>}
 deform_upperarm_bk_l {parent=deform_upperarm_bk_base_l, loc=<0 0 0>, rot=<0 0 0 1>}
 deform_upperarm_fr_base_l {parent=upperarm_twist_01_l, loc=<-5.6312 5.092 0.3354>, rot=<0 0 0 -1>}
 deform_upperarm_fr_l {parent=deform_upperarm_fr_base_l, loc=<0 0 0>, rot=<0 0 0 1>}
 deform_upperarm_up_base_l {parent=upperarm_twist_01_l, loc=<-8.891 0.4825 3.9517>, rot=<0 0 0 -1>}
 deform_upperarm_up_l {parent=deform_upperarm_up_base_l, loc=<0 0 0>, rot=<0 0 0 1>}
 mstr_armchains_l {parent=upperarm_twist_01_l, loc=<-7.2135 0.1134 0>, rot=<0 0 0 1>}
 base_armchains_low_l {parent=mstr_armchains_l, loc=<0 0 0>, rot=<0 0 0 1>}
 dyn_armchains_low_l {parent=base_armchains_low_l, loc=<0 0 0>, rot=<0 0 0 1>}
 base_armchains_mid_l {parent=mstr_armchains_l, loc=<0 0 0>, rot=<0 0 0 1>}
 dyn_armchains_mid_l {parent=base_armchains_mid_l, loc=<0 0 0>, rot=<0 0 0 1>}
 base_armchains_up_l {parent=mstr_armchains_l, loc=<0 0 0>, rot=<0 0 0 1>}
 dyn_armchains_up_l {parent=base_armchains_up_l, loc=<0 0 0>, rot=<0 0 0 1>}
 upperarm_twist_02_l {parent=upperarm_l, loc=<19.7998 -0.1134 0>, rot=<0 0 0 1>}
 base_backchains_upper_l {parent=clavicle_l, loc=<3.5432 -3.7438 3.9827>, rot=<0.1507 -0.6073 0.0698 -0.7769>}
 dyn_backchains_upper_l {parent=base_backchains_upper_l, loc=<0 0 0>, rot=<0 0 0 1>}
 clavicle_r {parent=spine_05, loc=<5.4831 -0.1053 3.2946>, rot=<0.7289 0.0465 -0.6829 0.009>}
 upperarm_r {parent=clavicle_r, loc=<-9.8658 1.2136 0.4105>, rot=<0.0164 -0.3245 -0.0269 -0.9454>}
 lowerarm_r {parent=upperarm_r, loc=<-26.465 0.8345 -0.259>, rot=<0.0036 0.0161 -0.2298 -0.9731>}
 hand_r {parent=lowerarm_r, loc=<-22.7355 0.3203 0.6106>, rot=<-0.6784 -0.0223 0.0735 -0.7307>}
 thumb_01_r {parent=hand_r, loc=<-1.3118 0.3191 2.459>, rot=<0.6546 -0.4042 0.0664 -0.6354>}
 thumb_02_r {parent=thumb_01_r, loc=<-5.9336 0.0448 -0.0401>, rot=<-0.0037 0.0213 0.298 -0.9543>}
 thumb_03_r {parent=thumb_02_r, loc=<-2.2223 -0.3682 0.0907>, rot=<0.0118 -0.013 0.0066 -0.9998>}
 index_metacarpal_r {parent=hand_r, loc=<-1.6535 -0.0183 2.96>, rot=<0.0158 -0.0124 -0.0173 -0.9996>}
 index_01_r {parent=index_metacarpal_r, loc=<-8.3008 -0.0357 0.0524>, rot=<-0.0044 0.0485 0.0924 -0.9945>}
 index_02_r {parent=index_01_r, loc=<-3.3587 0.3004 -0.0452>, rot=<-0.002 0.0027 0.3225 -0.9466>}
 index_03_r {parent=index_02_r, loc=<-2.5578 0.123 0.0252>, rot=<-0.002 0.0067 0.2558 -0.9667>}
 middle_metacarpal_r {parent=hand_r, loc=<-2.1856 0.1725 0.5984>, rot=<0.006 0.0168 -0.0329 -0.9993>}
 middle_01_r {parent=middle_metacarpal_r, loc=<-7.3275 0.2447 -0.0129>, rot=<-0.0208 0.0042 0.0858 -0.9961>}
 middle_02_r {parent=middle_01_r, loc=<-4.0123 -0.1401 -0.1175>, rot=<-0.0218 0.0258 0.3512 -0.9357>}
 middle_03_r {parent=middle_02_r, loc=<-2.6077 -0.0534 -0.1373>, rot=<0.0082 0.0188 0.1708 -0.9851>}
 ring_metacarpal_r {parent=hand_r, loc=<-2.2074 0.3625 -1.3638>, rot=<-0.0284 0.0722 -0.0458 -0.9959>}
 ring_01_r {parent=ring_metacarpal_r, loc=<-7.0295 0.6197 0.1446>, rot=<-0.0489 -0.0252 0.1892 -0.9804>}
 ring_02_r {parent=ring_01_r, loc=<-3.6174 -0.4647 0.1027>, rot=<0 0 0.2806 -0.9598>}
 ring_03_r {parent=ring_02_r, loc=<-2.4152 -0.114 -0.0199>, rot=<0 0 0.1189 -0.9929>}
 pinky_metacarpal_r {parent=hand_r, loc=<-1.7616 0.9753 -3.3814>, rot=<-0.0397 0.0814 -0.069 -0.9935>}
 pinky_01_r {parent=pinky_metacarpal_r, loc=<-7.1709 0.4831 0.1094>, rot=<-0.0654 -0.0035 0.2242 -0.9723>}
 pinky_02_r {parent=pinky_01_r, loc=<-2.8274 -0.0893 0.1043>, rot=<0.0037 0.0011 0.2854 -0.9584>}
 pinky_03_r {parent=pinky_02_r, loc=<-2.2046 -0.0307 0.0078>, rot=<-0.0002 0.0098 0.1445 -0.9895>}
 hand_attach_r {parent=hand_r, loc=<-8.5001 2.9999 0>, rot=<0 0 -0.7071 -0.7071>}
 deform_wrist_dn_base_r {parent=hand_r, loc=<-0.0001 1.5328 0>, rot=<0 0 0 -1>}
 deform_wrist_dn_r {parent=deform_wrist_dn_base_r, loc=<0 0 0>, rot=<0 0 0 1>}
 deform_wrist_up_base_r {parent=hand_r, loc=<-0.0002 -2.3007 -0.0001>, rot=<0 0 0 -1>}
 deform_wrist_up_r {parent=deform_wrist_up_base_r, loc=<0 0 0>, rot=<0 0 0 1>}
 lowerarm_twist_02_r {parent=lowerarm_r, loc=<-10.6707 0 0>, rot=<0 0 0 -1>}
 lowerarm_twist_01_r {parent=lowerarm_r, loc=<-18.316 0 0>, rot=<0 0 0 -1>}
 upperarm_twist_01_r {parent=upperarm_r, loc=<-7.2135 0.1134 0>, rot=<0 0 0 -1>}
 deform_upperarm_bk_base_r {parent=upperarm_twist_01_r, loc=<6.4053 4.3746 -0.314>, rot=<0 0 0 -1>}
 deform_upperarm_bk_r {parent=deform_upperarm_bk_base_r, loc=<0 0 0>, rot=<0 0 0 1>}
 deform_upperarm_fr_base_r {parent=upperarm_twist_01_r, loc=<5.6312 -5.092 -0.3354>, rot=<0 0 0 -1>}
 deform_upperarm_fr_r {parent=deform_upperarm_fr_base_r, loc=<0 0 0>, rot=<0 0 0 1>}
 deform_upperarm_up_base_r {parent=upperarm_twist_01_r, loc=<8.891 -0.4825 -3.9517>, rot=<0 0 0 -1>}
 deform_upperarm_up_r {parent=deform_upperarm_up_base_r, loc=<0 0 0>, rot=<0 0 0 1>}
 mstr_armchains_r {parent=upperarm_twist_01_r, loc=<7.2137 -0.1134 -0.0003>, rot=<0 0 0 1>}
 base_armchains_low_r {parent=mstr_armchains_r, loc=<0 0 0>, rot=<0 0 0 1>}
 dyn_armchains_low_r {parent=base_armchains_low_r, loc=<0 0 0>, rot=<0 0 0 1>}
 base_armchains_mid_r {parent=mstr_armchains_r, loc=<0 0 0>, rot=<0 0.1554 0 -0.9879>}
 dyn_armchains_mid_r {parent=base_armchains_mid_r, loc=<0 0 0>, rot=<0 0 0 1>}
 base_armchains_up_r {parent=mstr_armchains_r, loc=<0 0 0>, rot=<0 0 0 1>}
 dyn_armchains_up_r {parent=base_armchains_up_r, loc=<0 0 0>, rot=<0 0 0 1>}
 upperarm_twist_02_r {parent=upperarm_r, loc=<-19.7998 0.1134 0>, rot=<0 0 0 -1>}
 base_backchains_upper_r {parent=clavicle_r, loc=<-3.5432 3.7438 -3.983>, rot=<0.1507 -0.6073 0.0698 -0.7769>}
 dyn_backchains_upper_r {parent=base_backchains_upper_r, loc=<0 0 0>, rot=<0 0 0 1>}
 attach_backpack {parent=spine_05, loc=<-0.123 -13.473 0>, rot=<-0.0093 -0.707 0.0093 -0.707>}
 Backpack_BR {parent=attach_backpack, type=Null, loc=<0 7.2998 0>, rot=<0 0 0 1>}
 attach_cape {parent=spine_05, loc=<9.017 -11.714 0>, rot=<0.0616 -0.7044 -0.0616 -0.7044>}
 cape {parent=attach_cape, type=Null, loc=<0 5.5844 0>, rot=<0 0 0 1>}
 dyn_jacketjewelry_front_l {parent=spine_05, loc=<5.7922 3.6665 -11.8719>, rot=<-0.4813 -0.4364 -0.3546 -0.6725>}
 dyn_jacketjewelry_back_l {parent=spine_05, loc=<5.3048 -6.0354 -12.3759>, rot=<0.5189 -0.3915 0.3538 -0.6725>}
 dyn_jacketjewelry_back_r {parent=spine_05, loc=<5.3047 -6.0353 12.376>, rot=<-0.3915 -0.5189 0.6725 0.3538>}
 dyn_jacketjewelry_front_r {parent=spine_05, loc=<5.7918 3.6664 11.8719>, rot=<-0.4364 0.4813 0.6725 -0.3546>}
 base_coat_front_01_r {parent=spine_04, loc=<5.3512 9.1303 6.4004>, rot=<0 -0.7071 -0.0588 -0.7047>}
 dyn_coat_front_01_r {parent=base_coat_front_01_r, loc=<0 0 0>, rot=<0.0189 0.7069 0.0399 -0.706>}
 dyn_coat_front_02_r {parent=dyn_coat_front_01_r, loc=<-2.697 0 0>, rot=<0 0 0.0149 -0.9999>}
 base_coat_front_01_l {parent=spine_04, loc=<5.3516 9.1303 -6.4004>, rot=<0.7071 0 0.7047 -0.0588>}
 dyn_coat_front_01_l {parent=base_coat_front_01_l, loc=<0 0 0>, rot=<-0.0294 -0.7065 0.0294 -0.7065>}
 dyn_coat_front_02_l {parent=dyn_coat_front_01_l, loc=<2.6974 0 0>, rot=<0 0 0 1>}
 mstr_chestchains_r {parent=spine_04, loc=<11.1915 4.7743 9.0736>, rot=<0.9421 -0.2765 -0.0329 -0.1868>}
 base_chestchains_low_r {parent=mstr_chestchains_r, loc=<0 0 0>, rot=<0 0 0 1>}
 dyn_chestchains_low_r {parent=base_chestchains_low_r, loc=<0 0 0>, rot=<0 0 0 1>}
 base_chestchains_upper_r {parent=mstr_chestchains_r, loc=<0 0 0>, rot=<0 0 0 1>}
 dyn_chestchains_upper_r {parent=base_chestchains_upper_r, loc=<0 0 0>, rot=<0 0 0 1>}
 base_backchains_low_r {parent=spine_04, loc=<5.2713 -7.258 5.7091>, rot=<0.9952 0.0614 0.0628 0.0426>}
 dyn_backchains_low_r {parent=base_backchains_low_r, loc=<0 0 0>, rot=<0 0 0 1>}
 deform_pec_aim_base_l {parent=spine_04, loc=<3.2456 9.3309 -1.2237>, rot=<0.0603 -0.4033 0.3318 -0.8507>}
 deform_pec_aim_l {parent=deform_pec_aim_base_l, loc=<0 0 0>, rot=<0 0 0 1>}
 deform_pec_base_l {parent=deform_pec_aim_l, loc=<9.3665 0 0>, rot=<0 0 0 1>}
 deform_pec_l {parent=deform_pec_base_l, loc=<0 0 0>, rot=<0 0 0 1>}
 deform_pec_aim_base_r {parent=spine_04, loc=<3.2452 9.3309 1.2237>, rot=<-0.4033 -0.0603 0.8507 0.3318>}
 deform_pec_aim_r {parent=deform_pec_aim_base_r, loc=<0 0 0>, rot=<0 0 0 1>}
 deform_pec_base_r {parent=deform_pec_aim_r, loc=<-9.3667 -0.0003 -0.0004>, rot=<0 0 0 1>}
 deform_pec_r {parent=deform_pec_base_r, loc=<0 0 0>, rot=<0 0 0 1>}
 base_backchains_low_l {parent=spine_04, loc=<5.271 -7.258 -5.7091>, rot=<0.0614 -0.9952 -0.0426 0.0628>}
 dyn_backchains_low_l {parent=base_backchains_low_l, loc=<0 0 0>, rot=<0 0 0 1>}
 mstr_chestchains_l {parent=spine_04, loc=<11.1918 4.7743 -9.0736>, rot=<-0.2765 -0.9421 0.1868 -0.0329>}
 base_chestchains_low_l {parent=mstr_chestchains_l, loc=<0 0 0>, rot=<0 0 0 1>}
 dyn_chestchains_low_l {parent=base_chestchains_low_l, loc=<0 0 0>, rot=<0 0 0 1>}
 base_chestchains_upper_l {parent=mstr_chestchains_l, loc=<0 0 0>, rot=<0 0 0 1>}
 dyn_chestchains_upper_l {parent=base_chestchains_upper_l, loc=<0 0 0>, rot=<0 0 0 1>}
 thigh_l {parent=pelvis, loc=<-1.0855 -0.2109 -8.7053>, rot=<0.0009 0.0355 0.0155 -0.9993>}
 calf_l {parent=thigh_l, loc=<-42.5284 -0.0019 -0.7297>, rot=<0.0006 -0.0161 -0.0391 -0.9991>}
 foot_l {parent=calf_l, loc=<-46.7401 -2.5014 -0.592>, rot=<0.0427 -0.0238 0.0353 -0.9982>}
 ball_l {parent=foot_l, loc=<-3.5658 10.8066 -0.1544>, rot=<0.0001 -0.0001 -0.7063 -0.7079>}
 calf_twist_02_l {parent=calf_l, loc=<-19.1546 -1.7943 -0.2638>, rot=<0.0237 0.0095 -0.0035 -0.9997>}
 calf_twist_01_l {parent=calf_l, loc=<-35.8483 -1.9197 -0.5794>, rot=<0.0237 0.0095 -0.0035 -0.9997>}
 base_zipper_01_l {parent=calf_l, loc=<-23.6439 -2.4558 3.9145>, rot=<0.0194 -0.0023 0.999 0.0414>}
 dyn_zipper_01_l {parent=base_zipper_01_l, loc=<0 0 0>, rot=<0 0 0 1>}
 base_tie_mstr_l {parent=calf_l, loc=<-29 0 0>, rot=<0.3765 -0.017 0.9255 0.0378>}
 base_strand_in_01_l {parent=base_tie_mstr_l, loc=<0.1425 -2.5012 -0.2026>, rot=<0 -0.3588 0 -0.9334>}
 dyn_strand_in_01_l {parent=base_strand_in_01_l, loc=<0 0 0>, rot=<0 0 0 1>}
 base_strand_out_01_l {parent=base_tie_mstr_l, loc=<-0.587 -2.2992 -1.3312>, rot=<0 -0.3588 0 -0.9334>}
 dyn_strand_out_01_l {parent=base_strand_out_01_l, loc=<0 0 0>, rot=<0 0 0 1>}
 base_tie_out_01_l {parent=base_tie_mstr_l, loc=<-1.1103 -2.5457 -0.932>, rot=<0 -0.9187 0 -0.3951>}
 dyn_tie_out_01_l {parent=base_tie_out_01_l, loc=<0 0 0>, rot=<0 0 0 1>}
 base_tie_in_01_l {parent=base_tie_mstr_l, loc=<-0.3707 -2.6192 0.2186>, rot=<0 0.4207 0 -0.9072>}
 dyn_tie_in_01_l {parent=base_tie_in_01_l, loc=<0 0 0>, rot=<0 0 0 1>}
 deform_knee_in_base_l {parent=calf_l, loc=<0.8031 -5.0231 0.2456>, rot=<0 0 0 -1>}
 deform_knee_in_l {parent=deform_knee_in_base_l, loc=<0 0 0>, rot=<0 0 0 1>}
 deform_knee_out_base_l {parent=calf_l, loc=<0 4 0>, rot=<0 0 0 -1>}
 deform_knee_out_l {parent=deform_knee_out_base_l, loc=<0 0 0>, rot=<0 0 0 1>}
 thigh_twist_01_l {parent=thigh_l, loc=<-27.952 0 0>, rot=<0.0013 0 0 -1>}
 base_bk_skirt_l {parent=thigh_twist_01_l, loc=<25.6866 -9.4011 1.3142>, rot=<0.0178 -0.768 0.6395 0.0303>}
 dyn_bk_skirt_l {parent=base_bk_skirt_l, loc=<0 0 0>, rot=<0 0.0122 0 -0.9999>}
 base_side_skirt_l {parent=thigh_twist_01_l, loc=<25.1519 -0.4117 -5.6279>, rot=<-0.0048 -0.9991 0.0085 -0.042>}
 dyn_side_skirt_l {parent=base_side_skirt_l, loc=<0 0 0>, rot=<0 0.0122 0 -0.9999>}
 base_front_skirt_l {parent=thigh_twist_01_l, loc=<25.2714 -0.7228 -0.1272>, rot=<-0.0174 -0.7498 -0.6607 0.0315>}
 dyn_front_skirt_l {parent=base_front_skirt_l, loc=<0 0 0>, rot=<0 0.0122 0 -0.9999>}
 deform_groin_base_l {parent=thigh_l, loc=<0.4801 5.6776 -0.2486>, rot=<0 0 0 -1>}
 deform_groin_l {parent=deform_groin_base_l, loc=<0 0 0>, rot=<0 0 0 1>}
 thigh_r {parent=pelvis, loc=<-1.0855 -0.2109 8.7053>, rot=<0.0355 -0.0009 0.9993 0.0155>}
 calf_r {parent=thigh_r, loc=<42.5284 0.0019 0.7297>, rot=<0.0006 -0.0161 -0.0391 -0.9991>}
 foot_r {parent=calf_r, loc=<46.7401 2.5014 0.592>, rot=<0.0427 -0.0238 0.0353 -0.9982>}
 ball_r {parent=foot_r, loc=<3.5658 -10.8066 0.1544>, rot=<0.0001 -0.0001 -0.7063 -0.7079>}
 calf_twist_02_r {parent=calf_r, loc=<19.1546 1.7943 0.2638>, rot=<0.0237 0.0095 -0.0035 -0.9997>}
 calf_twist_01_r {parent=calf_r, loc=<35.8483 1.9197 0.5794>, rot=<0.0237 0.0095 -0.0035 -0.9997>}
 base_zipper_01_r {parent=calf_r, loc=<23.6439 2.4558 -3.9145>, rot=<0.0194 -0.0023 0.999 0.0414>}
 dyn_zipper_01_r {parent=base_zipper_01_r, loc=<0 0 0>, rot=<0 0 0 1>}
 base_tie_mstr_r {parent=calf_r, loc=<29 0 0>, rot=<0.3765 -0.017 0.9255 0.0378>}
 base_tie_in_01_r {parent=base_tie_mstr_r, loc=<0.3707 2.6193 -0.2186>, rot=<0 0.4207 0 -0.9072>}
 dyn_tie_in_01_r {parent=base_tie_in_01_r, loc=<0 0 0>, rot=<0 0 0 1>}
 base_tie_out_01_r {parent=base_tie_mstr_r, loc=<1.1103 2.5457 0.932>, rot=<0 -0.9187 0 -0.3951>}
 dyn_tie_out_01_r {parent=base_tie_out_01_r, loc=<0 0 0>, rot=<0 0 0 1>}
 base_strand_out_01_r {parent=base_tie_mstr_r, loc=<0.587 2.2992 1.3312>, rot=<0 -0.3588 0 -0.9334>}
 dyn_strand_out_01_r {parent=base_strand_out_01_r, loc=<0 0 0>, rot=<0 0 0 1>}
 base_strand_in_01_r {parent=base_tie_mstr_r, loc=<-0.1426 2.5012 0.2026>, rot=<0 -0.3588 0 -0.9334>}
 dyn_strand_in_01_r {parent=base_strand_in_01_r, loc=<0 0 0>, rot=<0 0 0 1>}
 deform_knee_in_base_r {parent=calf_r, loc=<-0.8031 5.0231 -0.2457>, rot=<0 0 0 -1>}
 deform_knee_in_r {parent=deform_knee_in_base_r, loc=<0 0 0>, rot=<0 0 0 1>}
 deform_knee_out_base_r {parent=calf_r, loc=<0 -4 0>, rot=<0 0 0 -1>}
 deform_knee_out_r {parent=deform_knee_out_base_r, loc=<0 0 0>, rot=<0 0 0 1>}
 thigh_twist_01_r {parent=thigh_r, loc=<27.952 0 0>, rot=<0.0013 0 0 -1>}
 base_bk_skirt_r {parent=thigh_twist_01_r, loc=<-25.68 9.7225 -1.3136>, rot=<0.6395 -0.0303 -0.0178 -0.768>}
 dyn_bk_skirt_r {parent=base_bk_skirt_r, loc=<0 0 0>, rot=<0 0.0122 0 -0.9999>}
 base_side_skirt_r {parent=thigh_twist_01_r, loc=<-25.1519 0.4117 5.628>, rot=<0.0085 0.042 0.0048 -0.9991>}
 dyn_side_skirt_r {parent=base_side_skirt_r, loc=<0 0 0>, rot=<0 0.0122 0 -0.9999>}
 base_front_skirt_r {parent=thigh_twist_01_r, loc=<-25.2695 0.7237 0.1277>, rot=<-0.6607 -0.0315 0.0174 -0.7498>}
 dyn_front_skirt_r {parent=base_front_skirt_r, loc=<0 0 0>, rot=<0 0.0122 0 -0.9999>}
 deform_groin_base_r {parent=thigh_r, loc=<-0.4801 -5.6776 0.2486>, rot=<0 0 0 -1>}
 deform_groin_r {parent=deform_groin_base_r, loc=<0 0 0>, rot=<0 0 0 1>}
 base_hip_side_upper_chain_01_r {parent=pelvis, loc=<11.4615 -1.8625 6.0595>, rot=<-0.6576 -0.2823 0.2841 -0.6381>}
 dyn_hip_side_upper_chain_01_r {parent=base_hip_side_upper_chain_01_r, loc=<0 0 0>, rot=<0 0 0 1>}
 base_hip_side_chain_01_r {parent=pelvis, loc=<10.4522 -1.8263 4.7522>, rot=<-0.6203 -0.2704 0.2625 -0.6878>}
 dyn_hip_side_chain_01_r {parent=base_hip_side_chain_01_r, loc=<0 0 0>, rot=<0 0 0 1>}
 dyn_hip_side_chain_02_r {parent=dyn_hip_side_chain_01_r, loc=<-7.8756 -0.0006 0>, rot=<0 0 0 1>}
 base_hip_side_upper_chain_01_l {parent=pelvis, loc=<11.4618 -1.8625 -6.0595>, rot=<0.2823 -0.6576 -0.6381 -0.2841>}
 dyn_hip_side_upper_chain_01_l {parent=base_hip_side_upper_chain_01_l, loc=<0 0 0>, rot=<0 0 0 1>}
 base_hip_side_chain_01_l {parent=pelvis, loc=<10.4518 -1.8263 -4.7522>, rot=<-0.2727 0.6143 0.6932 0.2602>}
 dyn_hip_side_chain_01_l {parent=base_hip_side_chain_01_l, loc=<0 0 0>, rot=<0 0 0 1>}
 dyn_hip_side_chain_02_l {parent=dyn_hip_side_chain_01_l, loc=<7.8418 -0.0099 -0.0076>, rot=<0 0 0 1>}
 base_hip_hoop_01_l {parent=pelvis, loc=<10.8366 4.6127 -6.419>, rot=<-0.0349 0.1541 0.9588 0.2362>}
 dyn_hip_hoop_01_l {parent=base_hip_hoop_01_l, loc=<0 0 0>, rot=<0 0 0 1>}
 base_hip_hoop_01_r {parent=pelvis, loc=<10.8371 4.6127 6.419>, rot=<-0.1541 -0.0349 0.2362 -0.9588>}
 dyn_hip_hoop_01_r {parent=base_hip_hoop_01_r, loc=<0 0 0>, rot=<0 0 0 1>}
 base_front_skirt_inner_l {parent=pelvis, loc=<3.378 5.3093 -6.4683>, rot=<-0.0056 -0.7515 -0.6597 -0.0054>}
 dyn_front_skirt_Inner_l {parent=base_front_skirt_inner_l, loc=<0 0 0>, rot=<0 0 0 1>}
 base_front_skirt_splitter_l {parent=dyn_front_skirt_Inner_l, loc=<0 0 0>, rot=<0 0 1 0>}
 dyn_front_skirt_splitter_l {parent=base_front_skirt_splitter_l, loc=<0 0 0>, rot=<0 0 0 1>}
 dyn_front_skirt_splitter_length_l {parent=dyn_front_skirt_splitter_l, loc=<0 0 0>, rot=<0 0 0 1>}
 dyn_front_skirt_Inner_length_l {parent=dyn_front_skirt_Inner_l, loc=<0 0 0>, rot=<0 0 0 1>}
 base_front_skirt_Inner_r {parent=pelvis, loc=<3.3776 5.5413 6.329>, rot=<0.0054 0.6597 0.7515 0.0056>}
 dyn_front_skirt_Inner_r {parent=base_front_skirt_Inner_r, loc=<0 0 0>, rot=<0 0 0 1>}
 base_front_skirt_splitter_r {parent=dyn_front_skirt_Inner_r, loc=<0 0 0>, rot=<0 0 0 1>}
 dyn_front_skirt_splitter_r {parent=base_front_skirt_splitter_r, loc=<0 0 0>, rot=<0 0 0 1>}
 dyn_front_skirt_splitter_length_r {parent=dyn_front_skirt_splitter_r, loc=<0 0 0>, rot=<0 0 0 1>}
 dyn_front_skirt_Inner_length_r {parent=dyn_front_skirt_Inner_r, loc=<0 0 0>, rot=<0 0 0 1>}
 thigh_l_null {parent=pelvis, type=Null, loc=<-1.0855 -0.2109 -8.7053>, rot=<0.0009 0.0355 0.0155 -0.9993>}
 thigh_r_null {parent=pelvis, type=Null, loc=<-1.0855 -0.2109 8.7053>, rot=<0.0355 -0.0009 0.9993 0.0155>}
 deform_glute_base_l {parent=pelvis, loc=<-7.3466 -9.0213 -9.1245>, rot=<0.0172 0.031 -0.4477 -0.8935>}
 deform_glute_l {parent=deform_glute_base_l, loc=<0 0 0>, rot=<0 0 0 1>}
 deform_glute_base_r {parent=pelvis, loc=<-7.3464 -9.0213 9.1245>, rot=<0.031 -0.0172 0.8935 -0.4477>}
 deform_glute_r {parent=deform_glute_base_r, loc=<0 0 0>, rot=<0 0 0 1>}
 ik_hand_root {parent=root, loc=<0 0 0>, rot=<0 0 0 1>}
 ik_hand_gun {parent=ik_hand_root, loc=<-48.1743 3.3545 104.4381>, rot=<0.6276 0.1309 0.4183 -0.6434>}
 ik_hand_r {parent=ik_hand_gun, loc=<0 0 0>, rot=<0 0 0 1>}
 ik_hand_l {parent=ik_hand_gun, loc=<59.3271 67.6997 34.3524>, rot=<-0.6158 -0.7027 -0.3565 0>}
 ik_foot_root {parent=root, loc=<0 0 0>, rot=<0 0 0 1>}
 ik_foot_l {parent=ik_foot_root, loc=<14.8262 -7.3545 8.296>, rot=<0.0357 0.7045 0.0284 -0.7082>}
 ik_foot_r {parent=ik_foot_root, loc=<-14.8262 -7.3545 8.296>, rot=<0.7082 0.0284 -0.7045 0.0357>}
]
variables [
 L thighBack Alpha {type=double, default=0}
 L thighForward Alpha {type=double, default=0}
 L thighSide Alpha {type=double, default=0}
 L thighSideIn Alpha {type=double, default=0}
 L thighSideIn_Bwd Alpha {type=double, default=0}
 L thighSideIn_Fwd Alpha {type=double, default=0}
 Parachuteopen {type=bool, default=False}
 R thighBack Alpha {type=double, default=0}
 R thighForward Alpha {type=double, default=0}
 R thighSide Alpha {type=double, default=0}
 R thighSideIn Alpha {type=double, default=0}
 R thighSideIn_Bwd Alpha {type=double, default=0}
 R thighSideIn_Fwd Alpha {type=double, default=0}
 isCrouching {type=bool, default=False}
 isEmoting {type=bool, default=False}
 isSkydive {type=bool, default=False}
]
